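# Translucent two-sided particle sprite material for streaks in the NightNight final-impact suck-in effect.
# Two panned turbulence textures, varied per particle, blend two chroma colors for emissive and multiply into opacity, shaped by a CenterPoint falloff and sine edge mask.
# A U-weighted offset vector points from each particle toward the ZP_positon world point, scaled by Length and per-particle randomness.
# Mobile uses actor position instead of particle position and skips depth fade.
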Material /FNGameplayCues/Athena/Prototype/Blueprints/NightNight/FX/FinalImpact/M_NightNight_SuckInFXBodyFlow
graph {
 properties {bUsedWithParticleSprites=true, blendMode=BLEND_Translucent, pixelDepthOffsetMode=PDOM_Legacy, twoSided=true, usedShadingModels=Default Lit}
 nodes [
  n1 = DynamicParameter() {paramNames [UV_Random, Length_Random, Param3, Param4]}
  Add(A=@n1.UV_Random)
  n2 = VectorParameter() {parameterName=Chroma1, defaultValue=<rgba 0.0659 0.8903 1 1>}
  n3 = VectorParameter() {parameterName=Chroma2, defaultValue=<rgba 1 0.6146 0.7047 1>}
  n4 = VectorParameter() {parameterName=Texture1_Ctrl, defaultValue=<rgba 0.2 0.4 0.8 0.02>}
  n5 = LinearInterpolate(Alpha=@n1.UV_Random) {constA=0.5, constB=1.5}
  n6 = Panner(Speed=Multiply(A=AppendVector(A=@n4.B, B=@n4.B), B=@n5))
  n7 = Multiply(A=AppendVector(A=@n4.R, B=@n4.R), B=@n5)
  n8 = DynamicParameter() {paramNames [UV_Random, Length_Random, Param3, Param4]}
  n9 = TextureSample(UVs=Add(A=Multiply(A=@n6, B=@n7), B=@n8.UV_Random)) {texture=/Game/Effects/Fort_Effects/Textures/Plasma/T_Turbulence_Stretched}
  n10 = VectorParameter() {parameterName=Texture2_Ctrl, defaultValue=<rgba 0.25 0.5 1 -0.01>}
  n11 = Panner(Speed=Multiply(A=AppendVector(A=@n10.B, B=@n10.B), B=@n5))
  n12 = Multiply(A=@n5, B=AppendVector(A=@n10.R, B=@n10.R))
  n13 = TextureSample(UVs=Add(A=@n8.UV_Random, B=Multiply(A=@n11, B=@n12))) {texture=&T_Turbulence_Stretched}
  n14 = Add(A=@n9.G, B=@n13.G)
  n15 = ParticleColor()
  MaterialFunctionCall(Emissive=Multiply(A=LinearInterpolate(A=@n2.RGB, B=@n3.RGB, Alpha=@n14), B=@n15.RGB)) {materialFunction=&MF_FX_EmissiveAdjustments}
  n16 = Subtract(
   A=ComponentMask(None=VectorParameter() {parameterName=ZP_positon, defaultValue=<rgba 66048 -43008 0 1>}.RGB) {b=true, g=true, r=true}
   B=ShadingPathSwitch(Default=ParticlePositionWS(), Mobile=ActorPositionWS()) {desc=Migrated from FeatureLevelSwitch}
  )
  n17 = ScalarParameter() {defaultValue=200, parameterName=Length}
  n18 = DynamicParameter() {paramNames [UV_Random, Length_Random, Param3, Param4]}
  n19 = TextureCoordinate()
  Multiply(
   A=Multiply(A=Normalize(VectorInput=@n16), B=Multiply(A=@n17, B=@n18.Length_Random)) {constB=100}
   B=OneMinus(None=ComponentMask(None=@n19) {r=true})
  )
  n20 = ComponentMask(None=TextureCoordinate()) {r=true}
  n21 = ScalarParameter() {defaultValue=0.7, parameterName=CenterPoint}
  n22 = Divide(B=OneMinus(None=@n21))
  n23 = Subtract(A=@n20, B=@n21) {constB=0.5}
  n24 = Divide(B=@n21)
  n25 = Abs(None=@n23)
  n26 = OneMinus(None=If(A=@n20, B=@n21, "A > B"=LinearInterpolate(B=@n22, Alpha=@n23), "A < B"=LinearInterpolate(B=@n24, Alpha=@n25)))
  n27 = TextureCoordinate()
  n28 = Sine(None=ComponentMask(None=Multiply(A=@n27) {constB=0.5}) {g=true})
  n29 = Multiply(A=Multiply(A=@n9.R, B=@n13.R), B=Multiply(A=Power(Base=@n26), B=Power(Base=@n28)))
  Multiply(A=@n15.A, B=ShadingPathSwitch(Default=DepthFade(Opacity=@n29) {fadeDistanceDefault=10}, Mobile=@n29) {desc=Migrated from FeatureLevelSwitch})
 ]
}
