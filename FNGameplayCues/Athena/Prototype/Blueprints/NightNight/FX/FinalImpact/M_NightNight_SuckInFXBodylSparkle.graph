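# Translucent, unlit sprite-particle material for sparkles in the NightNight final-impact suck-in effect.
# A cross-flare texture shapes opacity and blends a dark tint into the particle color, with a hot spherical-mask core added to the emissive.
# Depth and near-camera fades soften opacity, while low quality and mobile skip both fades and core.
# An offset vector aims from actor position toward the ZP_positon parameter, scaled by squared Position_Offset dynamic parameter.
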Material /FNGameplayCues/Athena/Prototype/Blueprints/NightNight/FX/FinalImpact/M_NightNight_SuckInFXBodylSparkle
graph {
 properties {
  bUsedWithParticleSprites=true
  blendMode=BLEND_Translucent
  pixelDepthOffsetMode=PDOM_Legacy
  shadingModel=MSM_Unlit
  twoSided=true
  usedShadingModels=Unlit
 }
 nodes [
  n1 = ParticleColor()
  n2 = StaticSwitchParameter(True=@n1.RGB, False=VectorParameter() {parameterName=Base Color, defaultValue=<rgba 0.6854 1.6198 2.8 1>}.RGB) {defaultValue=true, parameterName=Base Color}
  n3 = LinearInterpolate(A=@n2) {constAlpha=0.2}
  n4 = Power(Base=SphereMask(A=TextureCoordinate(), B=Constant() {r=0.5}) {attenuationRadius=0.08, hardnessPercent=0}) {constExponent=8}
  n5 = Constant()
  n6 = QualitySwitch(Default=Multiply(A=Multiply(A=@n3, B=@n4)) {constB=2}, Low=@n5)
  n7 = TextureSampleParameter2D() {parameterName=Texture, samplerType=SAMPLERTYPE_Grayscale, texture=/Game/Effects/Fort_Effects/Textures/Flares/T_Cross_Flare}
  n8 = LinearInterpolate(A=VectorParameter() {parameterName=DarkColor, defaultValue=<rgba 0.1133 0.088 0.2448 1>}.RGB, Alpha=@n7.R)
  MaterialFunctionCall(Emissive=Add(A=ShadingPathSwitch(Default=@n6, Mobile=@n5) {desc=Migrated from FeatureLevelSwitch}, B=Multiply(A=@n2, B=@n8))) {materialFunction=&MF_FX_EmissiveAdjustments}
  n9 = DepthFade(
   Opacity=MaterialFunctionCall(Distance=ScalarParameter() {defaultValue=150, parameterName=NearFadeDistance}) {materialFunction=&Near_Camera_Fade}.Result
   FadeDistance=ScalarParameter() {defaultValue=30, parameterName=DepthFadeDistance}
  )
  n10 = Constant() {r=1}
  Multiply(
   A=Multiply(A=@n1.A, B=@n7.R)
   B=ShadingPathSwitch(Default=QualitySwitch(Default=@n9, Low=@n10), Mobile=@n10) {desc=Migrated from FeatureLevelSwitch}
  )
  n11 = Subtract(
   A=ComponentMask(None=VectorParameter() {parameterName=ZP_positon, defaultValue=<rgba 66048 -43008 0 1>}.RGB) {b=true, g=true, r=true}
   B=ActorPositionWS()
  )
  n12 = DynamicParameter() {paramNames [Position_Offset, SpeedRandom, Param3, Param4]}
  Multiply(A=Multiply(A=Normalize(VectorInput=@n11)) {constB=100}, B=Multiply(A=@n12.Position_Offset, B=@n12.Position_Offset)) {constB=100}
 ]
}
