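# Translucent, unlit, two-sided sprite particle material for light streaks in the NightNight final-impact suck-in effect on players.
# Panning cellular noise, randomized by dynamic parameters and masked by a sine edge and CenterPoint ramp, drives tinted emissive and opacity.
# A U-weighted offset vector points toward ZP_positon to stretch streaks by randomized Length.
# Mobile skips depth and camera fades and measures direction from actor instead of particle position.
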Material /FNGameplayCues/Athena/Prototype/Blueprints/NightNight/FX/FinalImpact/M_NightNight_SuckInFXOnplayer_StreakLight
graph {
 properties {
  bUsedWithParticleSprites=true
  blendMode=BLEND_Translucent
  pixelDepthOffsetMode=PDOM_Legacy
  shadingModel=MSM_Unlit
  twoSided=true
  usedShadingModels=Unlit
 }
 nodes [
  n1 = ParticleColor()
  n2 = StaticComponentMaskParameter(None=Multiply(A=TextureCoordinate()) {constB=0.5}) {defaultG=true, parameterName=EdgeMask}
  n3 = VectorParameter() {parameterName=Texture_Ctrl, defaultValue=<rgba 0.02 0.5 4 0>}
  n4 = AppendVector(A=@n3.B, B=@n3.B)
  n5 = Add(
   A=DynamicParameter() {paramNames [UV_random, Length_Random, HV_random, Param4]}.UV_random
   B=Multiply(A=Panner(Speed=@n4), B=AppendVector(A=@n3.R, B=@n3.R))
  )
  n6 = Multiply(
   A=Power(Base=Sine(None=@n2))
   B=Power(
    Base=TextureSample(UVs=@n5) {samplerType=SAMPLERTYPE_Masks, texture=/Game/Global/Textures/Default/Noise/T_CellularNoise}.R
    Exp=ScalarParameter() {defaultValue=3, parameterName=Param_1}
   )
  )
  n7 = Multiply(A=Clamp(None=LinearInterpolate(Alpha=@n6) {constA=-0.5}), B=ScalarParameter() {defaultValue=10, parameterName=Intensity}) {constB=20}
  n8 = VectorParameter() {parameterName=HightColor, defaultValue=<rgba 1.0312 1.7902 3 1>}
  MaterialFunctionCall(Emissive=Add(A=@n1.RGB, B=Multiply(A=@n7, B=@n8.RGB))) {materialFunction=&MF_FX_EmissiveAdjustments}
  n9 = StaticComponentMaskParameter(None=TextureCoordinate()) {defaultR=true, parameterName=RampMask}
  n10 = ScalarParameter() {defaultValue=0.95, parameterName=CenterPoint}
  n11 = OneMinus(None=@n10)
  n12 = Subtract(A=@n9, B=@n10) {constB=0.5}
  n13 = If(A=@n9, B=@n10, "A > B"=LinearInterpolate(B=Divide(B=@n11), Alpha=@n12), "A < B"=LinearInterpolate(B=Divide(B=@n10), Alpha=Abs(None=@n12)))
  n14 = Multiply(A=@n6, B=Power(Base=OneMinus(None=@n13)))
  n15 = DepthFade(Opacity=@n14)
  n16 = MaterialFunctionCall(Fade Length=Constant() {r=20}) {materialFunction=&CameraDepthFade}
  Multiply(A=@n1.A, B=ShadingPathSwitch(Default=Multiply(A=@n15, B=@n16.Result), Mobile=@n14) {desc=Migrated from FeatureLevelSwitch})
  n17 = VectorParameter() {parameterName=ZP_positon, defaultValue=<rgba 66048 -43008 0 1>}
  n18 = ParticlePositionWS()
  n19 = ActorPositionWS()
  n20 = Normalize(VectorInput=Subtract(
    A=ComponentMask(None=@n17.RGB) {b=true, g=true, r=true}
    B=ShadingPathSwitch(Default=@n18, Mobile=@n19) {desc=Migrated from FeatureLevelSwitch}
   ))
  n21 = Multiply(
   A=ScalarParameter() {defaultValue=1000, parameterName=Length}
   B=DynamicParameter() {paramNames [UV_random, Length_Random, HV_random, Param4]}.Length_Random
  )
  n22 = ComponentMask(None=TextureCoordinate()) {r=true}
  Reroute(None=Multiply(A=Multiply(A=@n20, B=@n21) {constB=100}, B=OneMinus(None=@n22)))
 ]
}
